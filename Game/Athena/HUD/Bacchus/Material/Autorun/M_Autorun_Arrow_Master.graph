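# Translucent UI master material for the HUD autorun arrow icon, with three arrows packed into one texture's RGB channels.
# Each channel is sampled with scaled, vertically offset UVs driven by its position parameter.
# Opacity-weighted channels combine by Max into a master-opacity mask, while nested lerps tinted by Arrows_Color build colour.
# A separate static switch selects time-offset cosine looping or a static sum, which adds red twice and omits green.
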Material /Game/Athena/HUD/Bacchus/Material/Autorun/M_Autorun_Arrow_Master
graph {
 properties {blendMode=BLEND_Translucent, materialDomain=MD_UI, pixelDepthOffsetMode=PDOM_Legacy, usedShadingModels=Default Lit}
 nodes [
  n1 = Constant()
  n2 = LinearInterpolate(
   A=ScalarParameter() {defaultValue=-0.12, group=Icon, parameterName=Icon_RedChannel_StartPosition, sliderMax=0.12, sliderMin=-0.12}
   B=ScalarParameter() {group=Icon, parameterName=Icon_RedChannel_EndPosition, sliderMax=0.12, sliderMin=-0.12}
   Alpha=ScalarParameter() {defaultValue=1, group=Animation, parameterName=RedChannel_Position, sliderMax=1}
  ) {constA=0.35, constB=0.18}
  n3 = MaterialFunctionCall(UVs=TextureCoordinate(), Texture Scale=ScalarParameter() {defaultValue=0.9, group=Icon, parameterName=Icon_Scale}) {materialFunction=&ScaleUVsByCenter}
  n4 = TextureSampleParameter2D(UVs=Add(A=@n3.UVs, B=AppendVector(A=@n1, B=@n2)) {desc=Red Channel}) {desc=Red Channel, group=Icon, parameterName=Icon_Texture, texture=/Game/Athena/HUD/Bacchus/Material/Autorun/Textures/T_Autorun_Arrows}
  n5 = Constant()
  n6 = LinearInterpolate(
   A=ScalarParameter() {defaultValue=-0.12, group=Icon, parameterName=Icon_GreenChannel_StartPosition, sliderMax=0.12, sliderMin=-0.12}
   B=ScalarParameter() {group=Icon, parameterName=Icon_GreenChannel_EndPosition, sliderMax=0.12, sliderMin=-0.12}
   Alpha=ScalarParameter() {defaultValue=1, group=Animation, parameterName=GreenChannel_Position, sliderMax=1}
  ) {constA=0.35, constB=0.18}
  n7 = TextureSampleParameter2D(UVs=Add(A=@n3.UVs, B=AppendVector(A=@n5, B=@n6)) {desc=Green Channel}) {desc=Red Channel, group=Icon, parameterName=Icon_Texture, texture=&T_Autorun_Arrows}
  n8 = Max(
   A=Multiply(A=@n4.R, B=ScalarParameter() {defaultValue=0.25, group=Animation, parameterName=RedChannel_Opacity, sliderMax=1}) {desc=Red Channel}
   B=Multiply(A=@n7.G, B=ScalarParameter() {defaultValue=0.25, group=Animation, parameterName=GreenChannel_Opacity, sliderMax=1}) {desc=Green Channel}
  )
  n9 = Constant()
  n10 = LinearInterpolate(
   A=ScalarParameter() {defaultValue=-0.12, group=Icon, parameterName=Icon_BlueChannel_StartPosition, sliderMax=0.12, sliderMin=-0.12}
   B=ScalarParameter() {group=Icon, parameterName=Icon_BlueChannel_EndPosition, sliderMax=0.12, sliderMin=-0.12}
   Alpha=ScalarParameter() {defaultValue=1, group=Animation, parameterName=BlueChannel_Position, sliderMax=1}
  ) {constA=0.35, constB=0.18}
  n11 = TextureSampleParameter2D(UVs=Add(A=@n3.UVs, B=AppendVector(A=@n9, B=@n10)) {desc=Blue Channel}) {desc=Red Channel, group=Icon, parameterName=Icon_Texture, texture=&T_Autorun_Arrows}
  n12 = Multiply(A=@n11.B, B=ScalarParameter() {defaultValue=0.25, group=Animation, parameterName=BlueChannel_Opacity, sliderMax=1}) {desc=Blue Channel}
  Reroute(None=Multiply(A=Max(A=@n8, B=@n12), B=ScalarParameter() {defaultValue=1, parameterName=All_Opacity, sliderMax=1}))
  n13 = LinearInterpolate(A=@n4.R, B=@n7.G, Alpha=@n7.G)
  Reroute(None=Multiply(
    A=LinearInterpolate(A=@n13, B=@n11.B, Alpha=@n11.B)
    B=VectorParameter() {parameterName=Arrows_Color, defaultValue=<rgba 0.9131 0.6308 0.0802 1>}.RGB
   ))
  n14 = TextureSample() {texture=&T_Autorun_Arrows}
  n15 = TextureSample() {texture=&T_Autorun_Arrows}
  n16 = Multiply(A=Subtract(A=Time()) {constB=0.22}, B=ScalarParameter() {defaultValue=1, parameterName=TimeAnimMultiplier, sliderMax=2}) {constB=0.5}
  n17 = LinearInterpolate(B=@n15.R, Alpha=Clamp(None=Cosine(None=@n16))) {constAlpha=0, constB=0}
  n18 = Multiply(A=Time(), B=ScalarParameter() {defaultValue=1, parameterName=TimeAnimMultiplier, sliderMax=2}) {constB=0.5}
  n19 = LinearInterpolate(B=@n15.G, Alpha=Clamp(None=Cosine(None=@n18))) {constB=0}
  n20 = Subtract(A=Time()) {constB=0.8}
  n21 = ScalarParameter() {defaultValue=1, parameterName=TimeAnimMultiplier, sliderMax=2}
  n22 = Clamp(None=Cosine(None=Multiply(A=@n20, B=@n21) {constB=0.5}))
  StaticSwitchParameter(True=Add(A=Add(A=@n17, B=@n19), B=LinearInterpolate(B=@n15.B, Alpha=@n22) {constB=0}), False=Add(A=Add(A=@n14.R, B=@n14.R), B=@n14.B)) {parameterName=B_UseLoopingAnimation}
 ]
}
