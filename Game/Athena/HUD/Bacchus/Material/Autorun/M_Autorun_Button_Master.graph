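# Translucent UI master material that draws the HUD autorun button: an icon over a glowing, outlined background.
# SDF functions build a gradient-filled circle and an arc cut from an oversized, offset circle, with UseArc blending between them and fading the arc's sides.
# The icon's texture channels get separate opacities, and animation scalars tint icon and background toward the glow color.
# A ShowOnlyIcon static switch selects an icon-only variant.
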Material /Game/Athena/HUD/Bacchus/Material/Autorun/M_Autorun_Button_Master
graph {
 properties {blendMode=BLEND_Translucent, materialDomain=MD_UI, pixelDepthOffsetMode=PDOM_Legacy, usedShadingModels=Default Lit}
 nodes [
  n1 = TextureSampleParameter2D(UVs=NamedRerouteUsage()) {group=Icon, parameterName=Icon_Texture, texture=/Game/Athena/HUD/Bacchus/Material/Autorun/Textures/T_Autorun_Icon}
  n2 = LinearInterpolate(A=LinearInterpolate(A=@n1.R, B=@n1.R, Alpha=@n1.R), B=@n1.B, Alpha=@n1.B)
  NamedRerouteDeclaration(None=Multiply(A=@n2, B=NamedRerouteUsage())) {name=Icon, nodeColor=<rgba 0 0 0 1>}
  NamedRerouteDeclaration(None=TextureCoordinate()) {name=CircleUV, nodeColor=<rgba 0.0312 0 1 1>}
  NamedRerouteDeclaration(None=MaterialFunctionCall(
    UVs=NamedRerouteUsage()
    Size=ScalarParameter() {defaultValue=0.75, group=Glow FX, parameterName=GlowFX_Scale}
    OutlineThickness=ScalarParameter() {defaultValue=0.01, group=Glow FX, parameterName=GlowFX_OutlineThickness, sliderMax=0.1}
    OutlineAlignment=Constant() {r=-1}
    GlowMax=Constant() {r=0.04}
   ) {materialFunction=&MF_UI_Circle}.Outline) {name=CircleGlow, nodeColor=<rgba 0 1 0.5078 1>}
  n3 = Constant() {r=1}
  n4 = ScalarParameter() {defaultValue=0.75, group=Glow FX, parameterName=GlowFX_Scale}
  NamedRerouteDeclaration(None=MaterialFunctionCall!(
    DistanceField=NamedRerouteUsage()
    Threshold
    Thickness=ScalarParameter() {defaultValue=0.01, group=Glow FX, parameterName=GlowFX_OutlineThickness, sliderMax=0.1}
    Alignment=Divide(A=@n3, B=@n4)
    Smoothing
   ) {materialFunction=/FortUILibrary/MaterialFunctions/Shapes/Modifiers/MF_UI_ShapeOutline}.Outline) {name=ArcGlow, nodeColor=<rgba 0.4531 1 0 1>}
  n5 = MaterialFunctionCall(
   UVs=NamedRerouteUsage()
   Radius=Constant() {r=0.3}
   Density=ScalarParameter() {defaultValue=2.33, group=Background, parameterName=Gradient_Density, sliderMax=10}
  ) {materialFunction=&RadialGradientExponential}
  n6 = MaterialFunctionCall(
   UVs=NamedRerouteUsage()
   Size=Constant() {r=0.58}
   OutlineThickness=ScalarParameter() {defaultValue=0.02, group=Background, parameterName=OuterOutline_Width, sliderMax=0.35}
   OutlineAlignment=Constant() {r=1}
  ) {materialFunction=&MF_UI_Circle}
  n7 = NamedRerouteDeclaration(None=@n6.Fill) {name=CircleShape, nodeColor=<rgba 1 0 0.9375 1>}
  n8 = Multiply(
   A=Multiply(A=OneMinus(None=@n5.RadialGradientExponential), B=@n7)
   B=ScalarParameter() {defaultValue=0.5, group=Background, parameterName=Gradient_Opacity, sliderMax=1}
  )
  n9 = Multiply(A=@n7, B=ScalarParameter() {defaultValue=0.1, group=Background, parameterName=GradientFill_Opacity, sliderMax=1})
  NamedRerouteDeclaration(None=Saturate(None=Add(A=@n8, B=@n9))) {name=CircleBackground, nodeColor=<rgba 0 1 0.7422 1>}
  n10 = MaterialFunctionCall(UVs=NamedRerouteUsage(), Size=Constant() {r=3}, OutlineThickness=Constant() {r=0.6}) {materialFunction=&MF_UI_Circle}
  n11 = NamedRerouteDeclaration(None=@n10.Outline) {name=ArcShape, nodeColor=<rgba 1 0.5859 0 1>}
  n12 = NamedRerouteDeclaration(None=@n10.SDFOutline) {name=ArcSDF, nodeColor=<rgba 1 0.9844 0 1>}
  n13 = MaterialFunctionCall(DistanceField=@n12, Smoothing=Constant() {r=0.3}) {materialFunction=&MF_UI_AntiAliasing}
  n14 = Add(
   A=Multiply(A=Multiply(A=@n11, B=@n13.Outer), B=ScalarParameter() {defaultValue=0.5, group=Background, parameterName=Gradient_Opacity, sliderMax=1})
   B=Multiply(A=NamedRerouteUsage(), B=ScalarParameter() {defaultValue=0.1, group=Background, parameterName=GradientFill_Opacity, sliderMax=1})
  )
  NamedRerouteDeclaration(None=Saturate(None=LinearInterpolate(Alpha=@n14) {constB=2})) {name=ArcBackground, nodeColor=<rgba 0.3828 1 0 1>}
  n15 = LinearInterpolate(A=NamedRerouteUsage(), B=NamedRerouteUsage(), Alpha=ScalarParameter() {defaultValue=1, parameterName=UseArc, sliderMax=1})
  NamedRerouteDeclaration(None=Multiply(
    A=NamedRerouteDeclaration(None=@n15) {name=Glow, nodeColor=<rgba 0.4062 0 1 1>}
    B=VectorParameter() {group=Glow FX, parameterName=GlowFX_Color, defaultValue=<rgba 0.9131 0.6308 0.0802 1>}.RGB
   )) {name=GlowColor, nodeColor=<rgba 1 0 0.2109 1>}
  NamedRerouteDeclaration(None=@n6.Outline) {name=CircleOutline, nodeColor=<rgba 0.6641 0 1 1>}
  n16 = TextureSampleParameter2D(UVs=NamedRerouteUsage()) {group=Icon, parameterName=Icon_Texture, texture=&T_Autorun_Icon}
  n17 = Multiply(A=@n16.R, B=ScalarParameter() {defaultValue=0.151, group=Icon, parameterName=Icon_RedChannel_Opacity, sliderMax=1})
  n18 = Multiply(A=@n16.G, B=ScalarParameter() {defaultValue=0.5, group=Icon, parameterName=Icon_GreenChannel_Opacity, sliderMax=1})
  n19 = ScalarParameter() {defaultValue=1, group=Icon, parameterName=Icon_BlueChannel_Opacity, sliderMax=1}
  NamedRerouteDeclaration(None=Add(A=Add(A=@n17, B=@n18), B=Multiply(A=@n16.B, B=@n19))) {name=IconOpacity, nodeColor=<rgba 0 0 0 1>}
  NamedRerouteDeclaration(None=MaterialFunctionCall(
    DistanceField=@n12
    Thickness=ScalarParameter() {defaultValue=0.02, group=Background, parameterName=OuterOutline_Width, sliderMax=0.35}
    Alignment=Constant() {r=1}
   ) {materialFunction=&MF_UI_ShapeOutline}.Outline) {name=ArcOutline, nodeColor=<rgba 0.7812 1 0 1>}
  n20 = NamedRerouteDeclaration(None=LinearInterpolate(A=NamedRerouteUsage(), B=NamedRerouteUsage(), Alpha=ScalarParameter() {defaultValue=1, parameterName=UseArc, sliderMax=1})) {name=Outline, nodeColor=<rgba 0 0.8125 1 1>}
  n21 = LinearInterpolate(
   A=VectorParameter() {desc=Fill color, group=Background, parameterName=Fill_Color, defaultValue=<rgba 1 1 1 1>}.RGB
   B=VectorParameter() {group=Glow FX, parameterName=GlowFX_Color, defaultValue=<rgba 0.9131 0.6308 0.0802 1>}.RGB
   Alpha=ScalarParameter() {group=Background, parameterName=Animation_BG_Color, sliderMax=1}
  ) {constAlpha=0}
  n22 = NamedRerouteDeclaration(None=LinearInterpolate(A=NamedRerouteUsage(), B=NamedRerouteUsage(), Alpha=ScalarParameter() {defaultValue=1, parameterName=UseArc, sliderMax=1})) {name=Shape, nodeColor=<rgba 1 0.8672 0 1>}
  NamedRerouteDeclaration(None=Add(A=Multiply(A=@n20, B=@n21), B=Multiply(A=@n21, B=@n22))) {name=Color, nodeColor=<rgba 0.7813 0 1 1>}
  NamedRerouteDeclaration(None=MaterialFunctionCall(UVs=TextureCoordinate(), Texture Scale=ScalarParameter() {defaultValue=0.55, group=Icon, parameterName=Icon_Scale, sliderMax=1.5}) {materialFunction=&ScaleUVsByCenter}.UVs) {name=IconUV, nodeColor=<rgba 0.2891 0 1 1>}
  NamedRerouteDeclaration(None=LinearInterpolate(
    A=VectorParameter() {group=Icon, parameterName=IconStartColor, defaultValue=<rgba 1 1 1 1>}.RGB
    B=VectorParameter() {group=Glow FX, parameterName=GlowFX_Color, defaultValue=<rgba 0.9131 0.6308 0.0802 1>}.RGB
    Alpha=ScalarParameter() {group=Icon, parameterName=Animation_Icon_Color, sliderMax=1}
   ) {constA=0.35, constB=0.18}) {name=IconAnimation, nodeColor=<rgba 0 0.4375 1 1>}
  NamedRerouteDeclaration(None=Saturate(None=@n2)) {name=IconColor, nodeColor=<rgba 0.7578 1 0 1>}
  n23 = NamedRerouteUsage()
  n24 = NamedRerouteUsage()
  n25 = Subtract(A=Add(A=NamedRerouteUsage(), B=NamedRerouteUsage()), B=NamedRerouteUsage())
  NamedRerouteDeclaration(None=LinearInterpolate(A=LinearInterpolate(A=@n23, B=@n24, Alpha=@n25) {constAlpha=0}, B=NamedRerouteUsage(), Alpha=NamedRerouteUsage())) {name=FinalColor, nodeColor=<rgba 0.7344 0 1 1>}
  n26 = Multiply(A=NamedRerouteUsage(), B=ScalarParameter() {defaultValue=1, group=Icon, parameterName=Icon_OverallOpacity, sliderMax=1})
  n27 = NamedRerouteUsage()
  n28 = SmoothStep(Value=OneMinus(None=MaterialFunctionCall!(UV=@n27, "Arc Start Angle (0-1)", "Arc End Angle (0-1)", "Fill Ratio (0-1)", Fill Clockwise) {materialFunction=/FortUILibrary/MaterialFunctions/Basics/MF_UI_RadialArc}.Centered Fill)) {constMax=0.94, constMin=0.9}
  n29 = NamedRerouteUsage()
  n30 = ScalarParameter() {defaultValue=1, group=Glow FX, parameterName=GlowFX_Opacity, sliderMax=1}
  n31 = NamedRerouteUsage()
  n32 = ScalarParameter() {defaultValue=0.8, group=Background, parameterName=OuterOutline_Opacity, sliderMax=1}
  n33 = NamedRerouteDeclaration(None=LinearInterpolate(A=NamedRerouteUsage(), B=NamedRerouteUsage(), Alpha=ScalarParameter() {defaultValue=1, parameterName=UseArc, sliderMax=1})) {name=Background, nodeColor=<rgba 0 0.9531 1 1>}
  n34 = ScalarParameter() {defaultValue=1, group=Background, parameterName=Fill_Opacity, sliderMax=1}
  n35 = Saturate(None=Add(A=Multiply(A=@n31, B=@n32), B=Multiply(A=@n33, B=@n34)))
  n36 = ScalarParameter() {defaultValue=0.6, group=Background, parameterName=BG_OverallOpacity, sliderMax=1}
  n37 = Saturate(None=Add(A=Multiply(A=@n29, B=@n30), B=Multiply(A=@n35, B=@n36)))
  n38 = LinearInterpolate(
   A=@n37
   B=Multiply(A=@n37, B=NamedRerouteDeclaration(None=@n28) {name=SideFade, nodeColor=<rgba 0 1 0.9063 1>})
   Alpha=ScalarParameter() {defaultValue=1, parameterName=UseArc, sliderMax=1}
  )
  NamedRerouteDeclaration(None=Multiply(A=LinearInterpolate(A=@n26, B=@n37, Alpha=@n38), B=ScalarParameter() {defaultValue=1, parameterName=All_Opacity, sliderMax=1})) {name=FinalOpacity, nodeColor=<rgba 1 0.7266 0 1>}
  n39 = Constant()
  n40 = ScalarParameter() {defaultValue=-1.5, parameterName=BackgroundOffset}
  NamedRerouteDeclaration(None=Add(A=TextureCoordinate(), B=AppendVector(A=@n39, B=@n40))) {name=ArcUV, nodeColor=<rgba 0 1 0.0859 1>}
  StaticSwitchParameter(True=NamedRerouteUsage(), False=NamedRerouteUsage()) {parameterName=ShowOnlyIcon}
  StaticSwitchParameter(True=NamedRerouteUsage(), False=NamedRerouteUsage()) {parameterName=ShowOnlyIcon}
 ]
}
