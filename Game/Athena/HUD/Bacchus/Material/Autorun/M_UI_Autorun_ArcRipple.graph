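# Translucent UI material for the Bacchus HUD autorun element that draws tinted, animated ripples within an arc-shaped fade.
# An offset circle SDF drives both the MF_UI_Ripples gradient and a sine-based band, which reflected gradients shape into the opacity mask.
# Ripple phase comes from time since AnimStartTime, with speed, offset, loop and reverse controls.
# Debug switches allow a manual phase or show the mask alone.
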Material /Game/Athena/HUD/Bacchus/Material/Autorun/M_UI_Autorun_ArcRipple
graph {
 properties {bAutomaticallySetUsageInEditor=false, blendMode=BLEND_Translucent, materialDomain=MD_UI, usedShadingModels=Default Lit}
 nodes [
  n1 = NamedRerouteUsage()
  n2 = Subtract(A=ScalarParameter() {defaultValue=3, group=Animation, parameterName=LoopDuration, sortPriority=4}, B=@n1)
  NamedRerouteDeclaration(None=StaticSwitchParameter(
    True=ScalarParameter() {defaultValue=-2, group=Animation, parameterName=Phase}
    False=StaticSwitchParameter(True=@n2, False=@n1) {group=Animation, parameterName=Reverse}
   ) {group=Debug, parameterName=ManualPhase}) {name=Phase, nodeColor=<rgba 0 0.8275 1 1>}
  n3 = Subtract(A=Time() {period=1}, B=ScalarParameter() {group=Animation, parameterName=AnimStartTime, sortPriority=0}) {constA=38.8}
  n4 = ScalarParameter() {group=Animation, parameterName=StartOffset}
  n5 = Multiply(
   A=Add(A=Add(A=@n3, B=@n4), B=ScalarParameter() {group=Opacity, parameterName=StartSize, sliderMax=1})
   B=ScalarParameter() {defaultValue=1, group=Animation, parameterName=Speed, sortPriority=1}
  )
  n6 = Multiply(
   A=ScalarParameter() {defaultValue=3, group=Animation, parameterName=LoopDuration, sortPriority=4}
   B=ScalarParameter() {defaultValue=1, group=Animation, parameterName=Speed, sortPriority=1}
  )
  NamedRerouteDeclaration(None=StaticSwitchParameter(True=Fmod(A=@n5, B=@n6), False=@n5) {defaultValue=true, group=Animation, parameterName=Loop, sortPriority=3}) {name=Time, nodeColor=<rgba 0.3634 0.2387 1 1>}
  n7 = Constant()
  n8 = ScalarParameter() {defaultValue=-1.5, parameterName=BackgroundOffset}
  n9 = MaterialFunctionCall(UVs=Add(A=TextureCoordinate(), B=AppendVector(A=@n7, B=@n8)), Size=Constant() {r=3}, OutlineThickness=Constant() {r=0.6}) {materialFunction=&MF_UI_Circle}
  n10 = Sine(None=Multiply(A=@n9.SDFOutline) {constB=0.5})
  n11 = Power(Base=@n10) {constExponent=8}
  n12 = Multiply(
   A=Multiply(
    A=MaterialFunctionCall!(UVs, Offset, RotationCenter, RotationAngle) {materialFunction=/FortUILibrary/MaterialFunctions/Gradients/MF_UI_ReflectedGradient}.Result
    B=Subtract(A=@n10, B=@n11)
   )
   B=MaterialFunctionCall(RotationAngle=Constant() {r=0.25}) {materialFunction=&MF_UI_ReflectedGradient}.Result
  )
  n13 = Saturate(None=LinearInterpolate(Alpha=Saturate(None=@n12)) {constB=8})
  n14 = Add(A=@n9.SDFOutline)
  n15 = MaterialFunctionCall(
   DistanceField=MaterialFunctionCall!(
    UV
    Gradient=Multiply(A=@n14) {constB=0.5}
    Phase=NamedRerouteUsage()
    RippleCount=ScalarParameter() {defaultValue=3, group=Ripples, parameterName=RippleCount}
    Distribution=ScalarParameter() {defaultValue=1, group=Animation, parameterName=Distribution, sortPriority=2}
    Spacing=ScalarParameter() {defaultValue=0.5, group=Ripples, parameterName=Spacing}
    OutlineThickness=ScalarParameter() {defaultValue=0.1, group=Ripples, parameterName=OutlineThickness, sliderMax=1}
   ) {materialFunction=/Game/Athena/HUD/Bacchus/Material/RippleEffect/MF_UI_Ripples}.Gradient
   Threshold=Constant() {r=0.5}
  ) {materialFunction=&MF_UI_AntiAliasing}
  StaticSwitchParameter(
   True=@n13
   False=Multiply(A=Multiply(A=@n13, B=@n15.Outer), B=VectorParameter() {group=Ripples, parameterName=Color, defaultValue=<rgba 1 1 1 1>}.A)
  ) {group=Debug, parameterName=ShowOpacity}
  VectorParameter() {group=Ripples, parameterName=Color, defaultValue=<rgba 1 1 1 1>}
 ]
}
